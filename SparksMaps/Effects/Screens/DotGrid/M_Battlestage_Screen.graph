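# Animated battle stage screen material for Sparks, layering moving stripes and triangle shapes over a gradient background, tinted by Mojo stage colors.
# Shapes are procedural Step-based bar masks on sheared, panned and rotated UVs, with sine-driven offsets adding motion.
# An audio-analysis parameter collection supplies an amplitude signal for reactive modulation.
# The composite is multiplied by an adjustable blend of panning, distorted textures, then passed through the shared screen-details function.
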Material /SparksMaps/Effects/Screens/DotGrid/M_Battlestage_Screen
graph {
 expressions {
  MaterialExpressionAbs_0 {class=MaterialExpressionAbs, inputs [None {from=MaterialExpressionSine_1, output=""}], properties {}}
  MaterialExpressionAdd_0 {
   class=MaterialExpressionAdd
   inputs [
    A {from=MaterialExpressionNamedRerouteUsage_6}
    B {from=MaterialExpressionNamedRerouteUsage_7}
   ]
  }
  MaterialExpressionAdd_1 {
   class=MaterialExpressionAdd
   inputs [
    A {from=MaterialExpressionNamedRerouteUsage_0}
    B {from=MaterialExpressionMaterialFunctionCall_4, output=YX}
   ]
  }
  MaterialExpressionAdd_10 {
   class=MaterialExpressionAdd
   inputs [
    A {from=MaterialExpressionScalarParameter_15}
    B {from=""}
   ]
   properties {constA=0.75, constB=0.075}
  }
  MaterialExpressionAdd_11 {
   class=MaterialExpressionAdd
   inputs [
    A {from=MaterialExpressionReroute_50}
    B {from=MaterialExpressionMaterialFunctionCall_11, output=YX}
   ]
  }
  MaterialExpressionAdd_12 {
   class=MaterialExpressionAdd
   inputs [
    A {from=MaterialExpressionRotator_0}
    B {from=MaterialExpressionMaterialFunctionCall_13, output=YX}
   ]
  }
  MaterialExpressionAdd_13 {
   class=MaterialExpressionAdd
   inputs [
    A {from=MaterialExpressionScalarParameter_18}
    B {from=MaterialExpressionConstant_4}
   ]
   properties {constA=0.75, constB=0.075}
  }
  MaterialExpressionAdd_14 {
   class=MaterialExpressionAdd
   inputs [
    A {from=MaterialExpressionMultiply_25}
    B {from=MaterialExpressionLinearInterpolate_5}
   ]
  }
  MaterialExpressionAdd_16 {
   class=MaterialExpressionAdd
   inputs [
    A {from=MaterialExpressionMultiply_19}
    B {from=MaterialExpressionLinearInterpolate_6}
   ]
  }
  MaterialExpressionAdd_17 {
   class=MaterialExpressionAdd
   inputs [
    A {from=MaterialExpressionMaterialFunctionCall_25, output=Remapped Sine}
    B {from=MaterialExpressionMaterialFunctionCall_26, output=R}
   ]
  }
  MaterialExpressionAdd_18 {
   class=MaterialExpressionAdd
   inputs [
    A {from=MaterialExpressionMaterialFunctionCall_29, output=Remapped Sine}
    B {from=MaterialExpressionMaterialFunctionCall_30, output=R}
   ]
  }
  MaterialExpressionAdd_19 {
   class=MaterialExpressionAdd
   inputs [
    A {from=MaterialExpressionMultiply_71}
    B {from=MaterialExpressionNamedRerouteUsage_8}
   ]
  }
  MaterialExpressionAdd_2 {
   class=MaterialExpressionAdd
   inputs [
    A {from=MaterialExpressionRotator_1}
    B {from=MaterialExpressionAppendVector_2}
   ]
  }
  MaterialExpressionAdd_3 {
   class=MaterialExpressionAdd
   inputs [
    A {from=MaterialExpressionMultiply_12}
    B {from=MaterialExpressionMaterialFunctionCall_8, output=YX}
   ]
  }
  MaterialExpressionAdd_4 {
   class=MaterialExpressionAdd
   inputs [
    A {from=MaterialExpressionMaterialFunctionCall_2, output=R}
    B {from=MaterialExpressionConstantBiasScale_1}
   ]
  }
  MaterialExpressionAdd_5 {
   class=MaterialExpressionAdd
   inputs [
    A {from=MaterialExpressionReroute_18, output=Output}
    B {from=MaterialExpressionReroute_38, output=Output}
   ]
  }
  MaterialExpressionAdd_6 {
   class=MaterialExpressionAdd
   inputs [
    A {from=MaterialExpressionReroute_65}
    B {from=MaterialExpressionMaterialFunctionCall_5, output=YX}
   ]
  }
  MaterialExpressionAdd_8 {
   class=MaterialExpressionAdd
   inputs [
    A {from=MaterialExpressionReroute_66}
    B {from=MaterialExpressionMaterialFunctionCall_9, output=YX}
   ]
  }
  MaterialExpressionAdd_9 {
   class=MaterialExpressionAdd
   inputs [
    A {from=MaterialExpressionPanner_2}
    B {from=MaterialExpressionMultiply_11}
   ]
  }
  MaterialExpressionAppendVector_0 {
   class=MaterialExpressionAppendVector
   inputs [
    A {from=MaterialExpressionConstant4Vector_2, output=R}
    B {from=MaterialExpressionConstant4Vector_2, output=R}
   ]
  }
  MaterialExpressionAppendVector_1 {
   class=MaterialExpressionAppendVector
   inputs [
    A {from=MaterialExpressionConstant4Vector_3, output=R}
    B {from=MaterialExpressionConstant4Vector_3, output=R}
   ]
  }
  MaterialExpressionAppendVector_2 {
   class=MaterialExpressionAppendVector
   inputs [
    A {from=MaterialExpressionScalarParameter_7, output=""}
    B {from=MaterialExpressionScalarParameter_8, output=""}
   ]
  }
  MaterialExpressionAppendVector_3 {
   class=MaterialExpressionAppendVector
   inputs [
    A {from=MaterialExpressionConstant4Vector_1, output=B}
    B {from=MaterialExpressionConstant4Vector_1, output=B}
   ]
  }
  MaterialExpressionAppendVector_4 {
   class=MaterialExpressionAppendVector
   inputs [
    A {from=MaterialExpressionConstant4Vector_3, output=B}
    B {from=MaterialExpressionConstant4Vector_3, output=B}
   ]
  }
  MaterialExpressionAppendVector_5 {
   class=MaterialExpressionAppendVector
   inputs [
    A {from=MaterialExpressionScalarParameter_6, output=""}
    B {from=MaterialExpressionScalarParameter_23, output=""}
   ]
  }
  MaterialExpressionAppendVector_6 {
   class=MaterialExpressionAppendVector
   inputs [
    A {from=MaterialExpressionAdd_17, output=""}
    B {from=MaterialExpressionMaterialFunctionCall_26, output=G}
   ]
  }
  MaterialExpressionAppendVector_7 {
   class=MaterialExpressionAppendVector
   inputs [
    A {from=MaterialExpressionAdd_18, output=""}
    B {from=MaterialExpressionMaterialFunctionCall_30, output=G}
   ]
  }
  MaterialExpressionAppendVector_8 {
   class=MaterialExpressionAppendVector
   inputs [
    A {from=MaterialExpressionAdd_4, output=""}
    B {from=MaterialExpressionMaterialFunctionCall_2, output=G}
   ]
  }
  MaterialExpressionCollectionParameter_0 {
   class=MaterialExpressionCollectionParameter
   inputs []
   properties {collection=/FMAudioReactiveData/AudioAnalysis/MPC_Sparks_Audio_Parameters, parameterName=Amplitude}
  }
  MaterialExpressionComponentMask_0 {class=MaterialExpressionComponentMask, inputs [None {from=MaterialExpressionFrac_0}], properties {r=true}}
  MaterialExpressionComponentMask_1 {class=MaterialExpressionComponentMask, inputs [None {from=MaterialExpressionFrac_0}], properties {r=true}}
  MaterialExpressionComponentMask_11 {class=MaterialExpressionComponentMask, inputs [None {from=MaterialExpressionReroute_6}], properties {r=true}}
  MaterialExpressionComponentMask_13 {class=MaterialExpressionComponentMask, inputs [None {from=MaterialExpressionReroute_10}], properties {r=true}}
  MaterialExpressionComponentMask_15 {class=MaterialExpressionComponentMask, inputs [None {from=MaterialExpressionReroute_14}], properties {r=true}}
  MaterialExpressionComponentMask_20 {class=MaterialExpressionComponentMask, inputs [None {from=MaterialExpressionReroute_15, output=UV}], properties {r=true}}
  MaterialExpressionComponentMask_22 {class=MaterialExpressionComponentMask, inputs [None {from=MaterialExpressionReroute_30}], properties {r=true}}
  MaterialExpressionComponentMask_24 {class=MaterialExpressionComponentMask, inputs [None {from=MaterialExpressionReroute_35, output=UV}], properties {r=true}}
  MaterialExpressionComponentMask_26 {class=MaterialExpressionComponentMask, inputs [None {from=MaterialExpressionReroute_43}], properties {r=true}}
  MaterialExpressionComponentMask_28 {class=MaterialExpressionComponentMask, inputs [None {from=MaterialExpressionReroute_47}], properties {r=true}}
  MaterialExpressionComponentMask_30 {class=MaterialExpressionComponentMask, inputs [None {from=MaterialExpressionReroute_48}], properties {r=true}}
  MaterialExpressionComponentMask_32 {class=MaterialExpressionComponentMask, inputs [None {from=MaterialExpressionReroute_49}], properties {r=true}}
  MaterialExpressionComponentMask_35 {class=MaterialExpressionComponentMask, inputs [None {from=MaterialExpressionReroute_51}], properties {r=true}}
  MaterialExpressionComponentMask_37 {class=MaterialExpressionComponentMask, inputs [None {from=MaterialExpressionReroute_55}], properties {r=true}}
  MaterialExpressionComponentMask_40 {class=MaterialExpressionComponentMask, inputs [None {from=MaterialExpressionAdd_19}], properties {r=true}}
  MaterialExpressionComposite_10 {
   class=MaterialExpressionComposite
   inputs [
    UV {from=MaterialExpressionSubtract_5, output=""}
    CutOff {from=MaterialExpressionReroute_40, output=""}
   ]
   properties {subgraphName=BarMask}
  }
  MaterialExpressionComposite_11 {
   class=MaterialExpressionComposite
   inputs [
    UV {from=MaterialExpressionSubtract_5, output=""}
    CutOff {from=MaterialExpressionAdd_13, output=""}
   ]
   properties {subgraphName=BarMask}
  }
  MaterialExpressionComposite_12 {
   class=MaterialExpressionComposite
   inputs [
    UV {from=MaterialExpressionReroute_41, output=""}
    CutOff {from=MaterialExpressionAdd_13, output=""}
   ]
   properties {subgraphName=BarMask}
  }
  MaterialExpressionComposite_13 {
   class=MaterialExpressionComposite
   inputs [
    UV {from=MaterialExpressionSubtract_4, output=""}
    CutOff {from=MaterialExpressionReroute_40, output=""}
   ]
   properties {subgraphName=BarMask}
  }
  MaterialExpressionComposite_14 {
   class=MaterialExpressionComposite
   inputs [
    UV {from=MaterialExpressionReroute_62, output=""}
    CutOff {from=MaterialExpressionConstant_6, output=""}
   ]
   properties {subgraphName=BarMask}
  }
  MaterialExpressionComposite_15 {
   class=MaterialExpressionComposite
   inputs [
    UV {from=MaterialExpressionReroute_62, output=""}
    CutOff {from=MaterialExpressionConstant_7, output=""}
   ]
   properties {subgraphName=BarMask}
  }
  MaterialExpressionComposite_3 {
   class=MaterialExpressionComposite
   inputs [
    UV {from=MaterialExpressionSubtract_3, output=""}
    CutOff {from=MaterialExpressionReroute_19, output=""}
   ]
   properties {subgraphName=BarMask}
  }
  MaterialExpressionComposite_4 {
   class=MaterialExpressionComposite
   inputs [
    UV {from=MaterialExpressionSubtract_3, output=""}
    CutOff {from=MaterialExpressionAdd_10, output=""}
   ]
   properties {subgraphName=BarMask}
  }
  MaterialExpressionComposite_5 {
   class=MaterialExpressionComposite
   inputs [
    UV {from=MaterialExpressionReroute_20, output=""}
    CutOff {from=MaterialExpressionAdd_10, output=""}
   ]
   properties {subgraphName=BarMask}
  }
  MaterialExpressionComposite_7 {
   class=MaterialExpressionComposite
   inputs [
    UV {from=MaterialExpressionFrac_1, output=""}
    CutOffMax {from=MaterialExpressionConstant4Vector_0, output=R}
    CutOffMin {from=MaterialExpressionConstant4Vector_0, output=R}
   ]
   properties {subgraphName=BarMask}
  }
  MaterialExpressionComposite_8 {
   class=MaterialExpressionComposite
   inputs [
    UV {from=MaterialExpressionSubtract_1, output=""}
    CutOff {from=MaterialExpressionReroute_19, output=""}
   ]
   properties {subgraphName=BarMask}
  }
  MaterialExpressionComposite_9 {
   class=MaterialExpressionComposite
   inputs [
    UV {from=MaterialExpressionFrac_1, output=""}
    CutOffMax {from=MaterialExpressionConstant4Vector_0, output=B}
    CutOffMin {from=MaterialExpressionConstant4Vector_0, output=B}
   ]
   properties {subgraphName=BarMask}
  }
  MaterialExpressionConstant4Vector_0 {class=MaterialExpressionConstant4Vector, inputs [], properties {constant=<rgba 0.2 0.1 0.7 0.4>}}
  MaterialExpressionConstant4Vector_1 {class=MaterialExpressionConstant4Vector, inputs [], properties {constant=<rgba 0.5 0 0.15 0>}}
  MaterialExpressionConstant4Vector_2 {class=MaterialExpressionConstant4Vector, inputs [], properties {constant=<rgba 0.5 0 0.15 0>}}
  MaterialExpressionConstant4Vector_3 {class=MaterialExpressionConstant4Vector, inputs [], properties {constant=<rgba 0.1 0.1 0.05 -0.05>}}
  MaterialExpressionConstantBiasScale_1 {class=MaterialExpressionConstantBiasScale, inputs [None {from=MaterialExpressionSine_0}]}
  MaterialExpressionConstant_0 {class=MaterialExpressionConstant, properties {r=0.1}}
  MaterialExpressionConstant_1 {class=MaterialExpressionConstant, properties {r=-0.05}}
  MaterialExpressionConstant_10 {class=MaterialExpressionConstant, properties {r=0.85}}
  MaterialExpressionConstant_12 {class=MaterialExpressionConstant, properties {desc=SlantAmount, r=-0.75}}
  MaterialExpressionConstant_13 {class=MaterialExpressionConstant, properties {desc=SlantAmount, r=-0.75}}
  MaterialExpressionConstant_14 {class=MaterialExpressionConstant, properties {r=0.5}}
  MaterialExpressionConstant_15 {class=MaterialExpressionConstant, properties {r=0.2}}
  MaterialExpressionConstant_2 {class=MaterialExpressionConstant, properties {r=-0.06}}
  MaterialExpressionConstant_3 {class=MaterialExpressionConstant, properties {r=-0.01}}
  MaterialExpressionConstant_4 {class=MaterialExpressionConstant, properties {desc="RightTri_OuterSize ", r=0.0375}}
  MaterialExpressionConstant_5 {class=MaterialExpressionConstant, properties {r=12.57}}
  MaterialExpressionConstant_6 {class=MaterialExpressionConstant, properties {r=0.9}}
  MaterialExpressionConstant_7 {class=MaterialExpressionConstant, properties {r=0.87}}
  MaterialExpressionConstant_8 {class=MaterialExpressionConstant, properties {r=0.75}}
  MaterialExpressionConstant_9 {class=MaterialExpressionConstant, properties {r=-0.75}}
  MaterialExpressionDivide_0 {
   class=MaterialExpressionDivide
   inputs [
    A {from=MaterialExpressionScalarParameter_12}
    B {from=""}
   ]
  }
  MaterialExpressionDivide_1 {
   class=MaterialExpressionDivide
   inputs [
    A {from=MaterialExpressionConstant_12}
    B {from=""}
   ]
  }
  MaterialExpressionDivide_2 {
   class=MaterialExpressionDivide
   inputs [
    A {from=MaterialExpressionConstant_13}
    B {from=""}
   ]
  }
  MaterialExpressionDivide_3 {
   class=MaterialExpressionDivide
   inputs [
    A {from=MaterialExpressionScalarParameter_13}
    B {from=""}
   ]
  }
  MaterialExpressionDivide_4 {
   class=MaterialExpressionDivide
   inputs [
    A {from=MaterialExpressionConstant_9}
    B {from=""}
   ]
  }
  MaterialExpressionDivide_5 {
   class=MaterialExpressionDivide
   inputs [
    A {from=MaterialExpressionConstant_8}
    B {from=""}
   ]
  }
  MaterialExpressionDivide_6 {
   class=MaterialExpressionDivide
   inputs [
    A {from=MaterialExpressionScalarParameter_19}
    B {from=""}
   ]
   properties {constB=57.296}
  }
  MaterialExpressionFrac_0 {class=MaterialExpressionFrac, inputs [None {from=MaterialExpressionPanner_0, output=""}], properties {}}
  MaterialExpressionFrac_1 {class=MaterialExpressionFrac, inputs [None {from=MaterialExpressionPanner_1, output=""}], properties {}}
  MaterialExpressionLinearInterpolate_0 {
   class=MaterialExpressionLinearInterpolate
   inputs [
    A {from=""}
    B {from=MaterialExpressionLinearInterpolate_2}
    Alpha {from=MaterialExpressionScalarParameter_9}
   ]
   properties {constA=1}
  }
  MaterialExpressionLinearInterpolate_1 {
   class=MaterialExpressionLinearInterpolate
   inputs [
    A {from=MaterialExpressionLinearInterpolate_12}
    B {from=MaterialExpressionNamedRerouteUsage_13}
    Alpha {from=MaterialExpressionNamedRerouteUsage_9}
   ]
  }
  MaterialExpressionLinearInterpolate_11 {
   class=MaterialExpressionLinearInterpolate
   inputs [
    A {from=MaterialExpressionScalarParameter_37}
    B {from=MaterialExpressionScalarParameter_38}
    Alpha {from=MaterialExpressionMultiply_65}
   ]
   properties {constA=0.05, constB=0.75}
  }
  MaterialExpressionLinearInterpolate_12 {
   class=MaterialExpressionLinearInterpolate
   inputs [
    A {from=MaterialExpressionNamedRerouteUsage_10}
    B {from=MaterialExpressionLinearInterpolate_9}
    Alpha {from=MaterialExpressionNamedRerouteUsage_12}
   ]
  }
  MaterialExpressionLinearInterpolate_13 {
   class=MaterialExpressionLinearInterpolate
   inputs [
    A {from=""}
    B {from=MaterialExpressionNamedRerouteUsage_15}
    Alpha {from=MaterialExpressionScalarParameter_25}
   ]
   properties {constA=1}
  }
  MaterialExpressionLinearInterpolate_14 {
   class=MaterialExpressionLinearInterpolate
   inputs [
    A {from=""}
    B {from=MaterialExpressionNamedRerouteUsage_16}
    Alpha {from=MaterialExpressionScalarParameter_28}
   ]
   properties {constA=1}
  }
  MaterialExpressionLinearInterpolate_2 {
   class=MaterialExpressionLinearInterpolate
   inputs [
    A {from=MaterialExpressionTextureSample_1, output=R}
    B {from=MaterialExpressionTextureSample_3, output=R}
    Alpha {from=MaterialExpressionAppendVector_8}
   ]
  }
  MaterialExpressionLinearInterpolate_3 {
   class=MaterialExpressionLinearInterpolate
   inputs [
    A {from=MaterialExpressionNamedRerouteUsage_14}
    B {from=MaterialExpressionMaterialFunctionCall_23, output=Primary}
    Alpha {from=MaterialExpressionAdd_16}
   ]
   properties {desc=Comp Inner + Outer Tris}
  }
  MaterialExpressionLinearInterpolate_5 {
   class=MaterialExpressionLinearInterpolate
   inputs [
    A {from=""}
    B {from=MaterialExpressionMultiply_55}
    Alpha {from=MaterialExpressionScalarParameter_20}
   ]
   properties {desc=Outer Triangle Mask}
  }
  MaterialExpressionLinearInterpolate_6 {
   class=MaterialExpressionLinearInterpolate
   inputs [
    A {from=""}
    B {from=MaterialExpressionMultiply_56}
    Alpha {from=MaterialExpressionScalarParameter_20}
   ]
   properties {desc=Inner Triangle Mask}
  }
  MaterialExpressionLinearInterpolate_7 {
   class=MaterialExpressionLinearInterpolate
   inputs [
    A {from=""}
    B {from=MaterialExpressionMaterialFunctionCall_17, output=Primary}
    Alpha {from=MaterialExpressionReroute_0}
   ]
  }
  MaterialExpressionLinearInterpolate_8 {
   class=MaterialExpressionLinearInterpolate
   inputs [
    A {from=""}
    B {from=MaterialExpressionMaterialFunctionCall_18, output=Secondary}
    Alpha {from=MaterialExpressionNamedRerouteUsage_7}
   ]
  }
  MaterialExpressionLinearInterpolate_9 {
   class=MaterialExpressionLinearInterpolate
   inputs [
    A {from=MaterialExpressionLinearInterpolate_8}
    B {from=MaterialExpressionLinearInterpolate_7}
    Alpha {from=MaterialExpressionReroute_0}
   ]
  }
  MaterialExpressionMaterialFunctionCall_11 {
   class=MaterialExpressionMaterialFunctionCall
   inputs [
    XYZ {from=""}
    XY {from=MaterialExpressionMultiply_52}
   ]
   properties {materialFunction=&Swizzle}
  }
  MaterialExpressionMaterialFunctionCall_13 {
   class=MaterialExpressionMaterialFunctionCall
   inputs [
    XYZ {from=""}
    XY {from=MaterialExpressionMultiply_41}
   ]
   properties {materialFunction=&Swizzle}
  }
  MaterialExpressionMaterialFunctionCall_17 {
   class=MaterialExpressionMaterialFunctionCall
   inputs [CustomUPosition {from=""}]
   properties {materialFunction=/SparksStageShow/Systems/StageManager/FXManager/MaterialFunctions/MF_MojoStageColors}
  }
  MaterialExpressionMaterialFunctionCall_18 {
   class=MaterialExpressionMaterialFunctionCall
   inputs [CustomUPosition {from=""}]
   properties {materialFunction=&MF_MojoStageColors}
  }
  MaterialExpressionMaterialFunctionCall_2 {
   class=MaterialExpressionMaterialFunctionCall
   inputs [Float2 {from=MaterialExpressionNamedRerouteUsage_5}]
   properties {materialFunction=&BreakOutFloat2Components}
  }
  MaterialExpressionMaterialFunctionCall_22 {
   class=MaterialExpressionMaterialFunctionCall
   inputs [CustomUPosition {from=""}]
   properties {materialFunction=&MF_MojoStageColors}
  }
  MaterialExpressionMaterialFunctionCall_23 {
   class=MaterialExpressionMaterialFunctionCall
   inputs [CustomUPosition {from=""}]
   properties {materialFunction=&MF_MojoStageColors}
  }
  MaterialExpressionMaterialFunctionCall_24 {
   class=MaterialExpressionMaterialFunctionCall
   inputs [
    UVs {from=MaterialExpressionTextureCoordinate_0}
    Texture Scale {from=MaterialExpressionMultiply_64}
   ]
   properties {materialFunction=&ScaleUVsByCenter}
  }
  MaterialExpressionMaterialFunctionCall_25 {
   class=MaterialExpressionMaterialFunctionCall
   inputs [
    Sine Phase {from=MaterialExpressionMultiply_66}
    Value 1 {from=MaterialExpressionConstant_0}
    Value 2 {from=MaterialExpressionConstant_1}
   ]
   properties {materialFunction=/Engine/Functions/Engine_MaterialFunctions02/Utility/Sine_Remapped}
  }
  MaterialExpressionMaterialFunctionCall_26 {
   class=MaterialExpressionMaterialFunctionCall
   inputs [Float2 {from=MaterialExpressionNamedRerouteUsage_4}]
   properties {materialFunction=&BreakOutFloat2Components}
  }
  MaterialExpressionMaterialFunctionCall_27 {
   class=MaterialExpressionMaterialFunctionCall
   inputs [
    Emissive {from=MaterialExpressionMultiply_2}
    Scanlines Tile {from=""}
   ]
   properties {materialFunction=/SparksMaps/Effects/Screens/Materials/MF/MF_Sparks_Screen_Details}
  }
  MaterialExpressionMaterialFunctionCall_29 {
   class=MaterialExpressionMaterialFunctionCall
   inputs [
    Sine Phase {from=MaterialExpressionMultiply_10}
    Value 1 {from=MaterialExpressionConstant_2}
    Value 2 {from=MaterialExpressionConstant_3}
   ]
   properties {materialFunction=&Sine_Remapped}
  }
  MaterialExpressionMaterialFunctionCall_30 {
   class=MaterialExpressionMaterialFunctionCall
   inputs [Float2 {from=MaterialExpressionNamedRerouteUsage_21}]
   properties {materialFunction=&BreakOutFloat2Components}
  }
  MaterialExpressionMaterialFunctionCall_4 {
   class=MaterialExpressionMaterialFunctionCall
   inputs [
    XYZ {from=""}
    XY {from=MaterialExpressionMultiply_6}
   ]
   properties {materialFunction=&Swizzle}
  }
  MaterialExpressionMaterialFunctionCall_5 {
   class=MaterialExpressionMaterialFunctionCall
   inputs [
    XYZ {from=""}
    XY {from=MaterialExpressionMultiply_15}
   ]
   properties {materialFunction=&Swizzle}
  }
  MaterialExpressionMaterialFunctionCall_8 {
   class=MaterialExpressionMaterialFunctionCall
   inputs [
    XYZ {from=""}
    XY {from=MaterialExpressionMultiply_7}
   ]
   properties {materialFunction=&Swizzle}
  }
  MaterialExpressionMaterialFunctionCall_9 {
   class=MaterialExpressionMaterialFunctionCall
   inputs [
    XYZ {from=""}
    XY {from=MaterialExpressionMultiply_18}
   ]
   properties {materialFunction=&Swizzle}
  }
  MaterialExpressionMultiply_0 {
   class=MaterialExpressionMultiply
   inputs [
    A {from=MaterialExpressionStep_0}
    B {from=MaterialExpressionStep_1}
   ]
  }
  MaterialExpressionMultiply_1 {
   class=MaterialExpressionMultiply
   inputs [
    A {from=MaterialExpressionMultiply_0}
    B {from=MaterialExpressionConstant_10}
   ]
  }
  MaterialExpressionMultiply_10 {
   class=MaterialExpressionMultiply
   inputs [
    A {from=MaterialExpressionTime_3}
    B {from=MaterialExpressionScalarParameter_11}
   ]
  }
  MaterialExpressionMultiply_11 {
   class=MaterialExpressionMultiply
   inputs [
    A {from=MaterialExpressionTextureSample_3, output=G}
    B {from=MaterialExpressionScalarParameter_41}
   ]
   properties {constB=0.01}
  }
  MaterialExpressionMultiply_12 {
   class=MaterialExpressionMultiply
   inputs [
    A {from=MaterialExpressionNamedRerouteUsage_1}
    B {from=""}
   ]
   properties {constB=0.65}
  }
  MaterialExpressionMultiply_15 {
   class=MaterialExpressionMultiply
   inputs [
    A {from=MaterialExpressionReroute_65}
    B {from=MaterialExpressionScalarParameter_12}
   ]
  }
  MaterialExpressionMultiply_18 {
   class=MaterialExpressionMultiply
   inputs [
    A {from=MaterialExpressionReroute_66}
    B {from=MaterialExpressionScalarParameter_13}
   ]
  }
  MaterialExpressionMultiply_19 {
   class=MaterialExpressionMultiply
   inputs [
    A {from=MaterialExpressionReroute_2, output=Output}
    B {from=MaterialExpressionReroute_20, output=Output}
   ]
   properties {desc=Inner Triangle Mask}
  }
  MaterialExpressionMultiply_2 {
   class=MaterialExpressionMultiply
   inputs [
    A {from=MaterialExpressionMultiply_3}
    B {from=MaterialExpressionLinearInterpolate_0}
   ]
  }
  MaterialExpressionMultiply_25 {
   class=MaterialExpressionMultiply
   inputs [
    A {from=MaterialExpressionReroute_11, output=Output}
    B {from=MaterialExpressionReroute_5, output=Output}
   ]
   properties {desc=Outer Triangle Mask}
  }
  MaterialExpressionMultiply_3 {
   class=MaterialExpressionMultiply
   inputs [
    A {from=MaterialExpressionLinearInterpolate_1}
    B {from=MaterialExpressionMultiply_67}
   ]
  }
  MaterialExpressionMultiply_33 {
   class=MaterialExpressionMultiply
   inputs [
    A {from=MaterialExpressionStep_17}
    B {from=MaterialExpressionStep_16}
   ]
  }
  MaterialExpressionMultiply_34 {
   class=MaterialExpressionMultiply
   inputs [
    A {from=MaterialExpressionStep_19}
    B {from=MaterialExpressionStep_18}
   ]
  }
  MaterialExpressionMultiply_35 {
   class=MaterialExpressionMultiply
   inputs [
    A {from=MaterialExpressionStep_21}
    B {from=MaterialExpressionStep_20}
   ]
  }
  MaterialExpressionMultiply_37 {
   class=MaterialExpressionMultiply
   inputs [
    A {from=MaterialExpressionStep_24}
    B {from=MaterialExpressionStep_25}
   ]
  }
  MaterialExpressionMultiply_38 {
   class=MaterialExpressionMultiply
   inputs [
    A {from=MaterialExpressionStep_27}
    B {from=MaterialExpressionStep_26}
   ]
  }
  MaterialExpressionMultiply_39 {
   class=MaterialExpressionMultiply
   inputs [
    A {from=MaterialExpressionStep_28}
    B {from=MaterialExpressionStep_29}
   ]
  }
  MaterialExpressionMultiply_41 {
   class=MaterialExpressionMultiply
   inputs [
    A {from=MaterialExpressionRotator_0}
    B {from=MaterialExpressionConstant_8}
   ]
  }
  MaterialExpressionMultiply_42 {
   class=MaterialExpressionMultiply
   inputs [
    A {from=MaterialExpressionReroute_29, output=Output}
    B {from=MaterialExpressionReroute_17, output=Output}
   ]
   properties {desc=Inner Triangle Mask}
  }
  MaterialExpressionMultiply_43 {
   class=MaterialExpressionMultiply
   inputs [
    A {from=MaterialExpressionReroute_13, output=Output}
    B {from=MaterialExpressionReroute_5, output=Output}
   ]
   properties {desc=Outer Triangle Mask}
  }
  MaterialExpressionMultiply_45 {
   class=MaterialExpressionMultiply
   inputs [
    A {from=MaterialExpressionStep_31}
    B {from=MaterialExpressionStep_30}
   ]
  }
  MaterialExpressionMultiply_46 {
   class=MaterialExpressionMultiply
   inputs [
    A {from=MaterialExpressionStep_33}
    B {from=MaterialExpressionStep_32}
   ]
  }
  MaterialExpressionMultiply_47 {
   class=MaterialExpressionMultiply
   inputs [
    A {from=MaterialExpressionStep_35}
    B {from=MaterialExpressionStep_34}
   ]
  }
  MaterialExpressionMultiply_48 {
   class=MaterialExpressionMultiply
   inputs [
    A {from=MaterialExpressionStep_37}
    B {from=MaterialExpressionStep_36}
   ]
  }
  MaterialExpressionMultiply_52 {
   class=MaterialExpressionMultiply
   inputs [
    A {from=MaterialExpressionReroute_50}
    B {from=MaterialExpressionConstant_9}
   ]
  }
  MaterialExpressionMultiply_53 {
   class=MaterialExpressionMultiply
   inputs [
    A {from=MaterialExpressionStep_39}
    B {from=MaterialExpressionStep_38}
   ]
  }
  MaterialExpressionMultiply_55 {
   class=MaterialExpressionMultiply
   inputs [
    A {from=MaterialExpressionMultiply_43}
    B {from=MaterialExpressionReroute_46, output=Output}
   ]
  }
  MaterialExpressionMultiply_56 {
   class=MaterialExpressionMultiply
   inputs [
    A {from=MaterialExpressionReroute_54, output=Output}
    B {from=MaterialExpressionMultiply_42}
   ]
  }
  MaterialExpressionMultiply_57 {
   class=MaterialExpressionMultiply
   inputs [
    A {from=MaterialExpressionStep_41}
    B {from=MaterialExpressionStep_40}
   ]
  }
  MaterialExpressionMultiply_58 {
   class=MaterialExpressionMultiply
   inputs [
    A {from=MaterialExpressionDivide_6}
    B {from=""}
   ]
   properties {constB=4, desc=Lets UV_Angle input be in degrees}
  }
  MaterialExpressionMultiply_6 {
   class=MaterialExpressionMultiply
   inputs [
    A {from=MaterialExpressionNamedRerouteUsage_0}
    B {from=MaterialExpressionConstant_12}
   ]
  }
  MaterialExpressionMultiply_62 {
   class=MaterialExpressionMultiply
   inputs [
    A {from=MaterialExpressionMaterialFunctionCall_22, output=Secondary}
    B {from=MaterialExpressionLinearInterpolate_11}
   ]
  }
  MaterialExpressionMultiply_64 {
   class=MaterialExpressionMultiply
   inputs [
    A {from=MaterialExpressionScalarParameter_24}
    B {from=MaterialExpressionAppendVector_5}
   ]
  }
  MaterialExpressionMultiply_65 {
   class=MaterialExpressionMultiply
   inputs [
    A {from=MaterialExpressionPower_0}
    B {from=MaterialExpressionLinearInterpolate_13}
   ]
  }
  MaterialExpressionMultiply_66 {
   class=MaterialExpressionMultiply
   inputs [
    A {from=MaterialExpressionTime_0}
    B {from=MaterialExpressionScalarParameter_29}
   ]
  }
  MaterialExpressionMultiply_67 {
   class=MaterialExpressionMultiply
   inputs [
    A {from=MaterialExpressionScalarParameter_0}
    B {from=MaterialExpressionLinearInterpolate_14}
   ]
  }
  MaterialExpressionMultiply_7 {
   class=MaterialExpressionMultiply
   inputs [
    A {from=MaterialExpressionMultiply_12}
    B {from=MaterialExpressionConstant_13}
   ]
  }
  MaterialExpressionMultiply_71 {
   class=MaterialExpressionMultiply
   inputs [
    A {from=MaterialExpressionTime_1}
    B {from=MaterialExpressionScalarParameter_36}
   ]
  }
  MaterialExpressionMultiply_8 {
   class=MaterialExpressionMultiply
   inputs [
    A {from=MaterialExpressionNamedRerouteUsage_2}
    B {from=MaterialExpressionConstant_14}
   ]
   properties {constB=0.65}
  }
  MaterialExpressionMultiply_9 {
   class=MaterialExpressionMultiply
   inputs [
    A {from=MaterialExpressionTime_2}
    B {from=MaterialExpressionConstant_15}
   ]
  }
  MaterialExpressionNamedRerouteDeclaration_0 {
   class=MaterialExpressionNamedRerouteDeclaration
   inputs [None {from=MaterialExpressionAdd_2}]
   properties {name=UV_Master, nodeColor=<rgba 1 0 0.1792 1>}
  }
  MaterialExpressionNamedRerouteDeclaration_1 {
   class=MaterialExpressionNamedRerouteDeclaration
   inputs [None {from=MaterialExpressionMultiply_62}]
   properties {name=BG_Gradient, nodeColor=<rgba 1 0.1172 0 1>}
  }
  MaterialExpressionNamedRerouteDeclaration_2 {
   class=MaterialExpressionNamedRerouteDeclaration
   inputs [None {from=MaterialExpressionSaturate_0}]
   properties {name=Stripes_Mask, nodeColor=<rgba 0 0.2031 1 1>}
  }
  MaterialExpressionNamedRerouteDeclaration_3 {
   class=MaterialExpressionNamedRerouteDeclaration
   inputs [None {from=MaterialExpressionAdd_14}]
   properties {name=Shape_Triangles_Mask, nodeColor=<rgba 0.5234 1 0 1>}
  }
  MaterialExpressionNamedRerouteDeclaration_4 {
   class=MaterialExpressionNamedRerouteDeclaration
   inputs [None {from=MaterialExpressionLinearInterpolate_3}]
   properties {name=Shape_Triangles_RGB, nodeColor=<rgba 1 0.1875 0 1>}
  }
  MaterialExpressionNamedRerouteDeclaration_5 {
   class=MaterialExpressionNamedRerouteDeclaration
   inputs [None {from=MaterialExpressionCollectionParameter_0}]
   properties {name=AmplitudeSignal, nodeColor=<rgba 0.1953 1 0 1>}
  }
  MaterialExpressionNamedRerouteDeclaration_6 {
   class=MaterialExpressionNamedRerouteDeclaration
   inputs [None {from=MaterialExpressionMultiply_1}]
   properties {name=Shape_Stripe01, nodeColor=<rgba 1 0 0.8906 1>}
  }
  MaterialExpressionNamedRerouteDeclaration_7 {
   class=MaterialExpressionNamedRerouteDeclaration
   inputs [None {from=MaterialExpressionAdd_5}]
   properties {name=Shape_Stripe02, nodeColor=<rgba 0 1 0.7891 1>}
  }
  MaterialExpressionNamedRerouteUsage_0 {class=MaterialExpressionNamedRerouteUsage}
  MaterialExpressionNamedRerouteUsage_1 {class=MaterialExpressionNamedRerouteUsage}
  MaterialExpressionNamedRerouteUsage_10 {class=MaterialExpressionNamedRerouteUsage}
  MaterialExpressionNamedRerouteUsage_12 {class=MaterialExpressionNamedRerouteUsage}
  MaterialExpressionNamedRerouteUsage_13 {class=MaterialExpressionNamedRerouteUsage}
  MaterialExpressionNamedRerouteUsage_14 {class=MaterialExpressionNamedRerouteUsage}
  MaterialExpressionNamedRerouteUsage_15 {class=MaterialExpressionNamedRerouteUsage}
  MaterialExpressionNamedRerouteUsage_16 {class=MaterialExpressionNamedRerouteUsage}
  MaterialExpressionNamedRerouteUsage_2 {class=MaterialExpressionNamedRerouteUsage}
  MaterialExpressionNamedRerouteUsage_21 {class=MaterialExpressionNamedRerouteUsage}
  MaterialExpressionNamedRerouteUsage_4 {class=MaterialExpressionNamedRerouteUsage}
  MaterialExpressionNamedRerouteUsage_5 {class=MaterialExpressionNamedRerouteUsage}
  MaterialExpressionNamedRerouteUsage_6 {class=MaterialExpressionNamedRerouteUsage}
  MaterialExpressionNamedRerouteUsage_7 {class=MaterialExpressionNamedRerouteUsage}
  MaterialExpressionNamedRerouteUsage_8 {class=MaterialExpressionNamedRerouteUsage}
  MaterialExpressionNamedRerouteUsage_9 {class=MaterialExpressionNamedRerouteUsage}
  MaterialExpressionPanner_0 {
   class=MaterialExpressionPanner
   inputs [
    Coordinate {from=MaterialExpressionSubtract_0}
    Time {from=""}
    Speed {from=MaterialExpressionAppendVector_0}
   ]
   properties {speedX=0.5}
  }
  MaterialExpressionPanner_1 {
   class=MaterialExpressionPanner
   inputs [
    Coordinate {from=MaterialExpressionSubtract_2}
    Time {from=""}
    Speed {from=MaterialExpressionAppendVector_3}
   ]
   properties {speedX=0.5}
  }
  MaterialExpressionPanner_2 {
   class=MaterialExpressionPanner
   inputs [
    Coordinate {from=MaterialExpressionNamedRerouteUsage_2}
    Time {from=""}
    Speed {from=MaterialExpressionAppendVector_1}
   ]
  }
  MaterialExpressionPanner_3 {
   class=MaterialExpressionPanner
   inputs [
    Coordinate {from=MaterialExpressionMultiply_8}
    Time {from=""}
    Speed {from=MaterialExpressionAppendVector_4}
   ]
  }
  MaterialExpressionPinBase_0 {class=MaterialExpressionPinBase, inputs [], properties {}}
  MaterialExpressionPinBase_1 {class=MaterialExpressionPinBase, inputs [], properties {}}
  MaterialExpressionPinBase_10 {
   class=MaterialExpressionPinBase
   inputs []
   properties {reroutePins [
     UV {expression=&self:MaterialGraph_0.MaterialGraphNode_Composite_11.BarMask.MaterialExpressionReroute_3}
     CutOff {expression=&self:MaterialGraph_0.MaterialGraphNode_Composite_11.BarMask.MaterialExpressionReroute_4}
    ]}
  }
  MaterialExpressionPinBase_11 {
   class=MaterialExpressionPinBase
   inputs [Output {from=MaterialExpressionMultiply_34, output=""}]
   properties {reroutePins [Output {expression=&self:MaterialGraph_0.MaterialGraphNode_Composite_11.BarMask.MaterialExpressionReroute_5}]}
  }
  MaterialExpressionPinBase_12 {
   class=MaterialExpressionPinBase
   inputs []
   properties {reroutePins [
     UV {expression=&self:MaterialGraph_0.MaterialGraphNode_Composite_16.BarMask.MaterialExpressionReroute_9}
     CutOff {expression=&self:MaterialGraph_0.MaterialGraphNode_Composite_16.BarMask.MaterialExpressionReroute_10}
    ]}
  }
  MaterialExpressionPinBase_13 {
   class=MaterialExpressionPinBase
   inputs [Output {from=MaterialExpressionMultiply_35, output=""}]
   properties {reroutePins [Output {expression=&self:MaterialGraph_0.MaterialGraphNode_Composite_16.BarMask.MaterialExpressionReroute_11}]}
  }
  MaterialExpressionPinBase_14 {
   class=MaterialExpressionPinBase
   inputs []
   properties {reroutePins [
     UV {expression=&self:MaterialGraph_0.MaterialGraphNode_Composite_7.MaterialGraph_0.MaterialExpressionReroute_9}
     CutOffMax {expression=&self:MaterialGraph_0.MaterialGraphNode_Composite_7.MaterialGraph_0.MaterialExpressionReroute_10}
     CutOffMin {expression=&self:MaterialGraph_0.MaterialGraphNode_Composite_7.MaterialGraph_0.MaterialExpressionReroute_11}
    ]}
  }
  MaterialExpressionPinBase_15 {
   class=MaterialExpressionPinBase
   inputs [Output {from="", output=""}]
   properties {reroutePins [Output {expression=&self:MaterialGraph_0.MaterialGraphNode_Composite_7.MaterialGraph_0.MaterialExpressionReroute_12}]}
  }
  MaterialExpressionPinBase_16 {
   class=MaterialExpressionPinBase
   inputs []
   properties {reroutePins [
     UV {expression=&self:MaterialGraph_0.MaterialGraphNode_Composite_9.BarMask.MaterialExpressionReroute_15}
     CutOffMax {expression=&self:MaterialGraph_0.MaterialGraphNode_Composite_9.BarMask.MaterialExpressionReroute_16}
     CutOffMin {expression=&self:MaterialGraph_0.MaterialGraphNode_Composite_9.BarMask.MaterialExpressionReroute_17}
    ]}
  }
  MaterialExpressionPinBase_17 {
   class=MaterialExpressionPinBase
   inputs [Output {from=MaterialExpressionMultiply_37, output=""}]
   properties {reroutePins [Output {expression=&self:MaterialGraph_0.MaterialGraphNode_Composite_9.BarMask.MaterialExpressionReroute_18}]}
  }
  MaterialExpressionPinBase_18 {
   class=MaterialExpressionPinBase
   inputs []
   properties {reroutePins [
     UV {expression=&self:MaterialGraph_0.MaterialGraphNode_Composite_13.BarMask.MaterialExpressionReroute_0}
     CutOff {expression=&self:MaterialGraph_0.MaterialGraphNode_Composite_13.BarMask.MaterialExpressionReroute_1}
    ]}
  }
  MaterialExpressionPinBase_19 {
   class=MaterialExpressionPinBase
   inputs [Output {from=MaterialExpressionMultiply_38, output=""}]
   properties {reroutePins [Output {expression=&self:MaterialGraph_0.MaterialGraphNode_Composite_13.BarMask.MaterialExpressionReroute_2}]}
  }
  MaterialExpressionPinBase_2 {class=MaterialExpressionPinBase, inputs [], properties {}}
  MaterialExpressionPinBase_20 {
   class=MaterialExpressionPinBase
   inputs []
   properties {reroutePins [
     UV {expression=&self:MaterialExpressionReroute_35}
     CutOffMax {expression=&self:MaterialExpressionReroute_36}
     CutOffMin {expression=&self:MaterialExpressionReroute_37}
    ]}
  }
  MaterialExpressionPinBase_21 {
   class=MaterialExpressionPinBase
   inputs [Output {from=MaterialExpressionMultiply_39, output=""}]
   properties {reroutePins [Output {expression=&self:MaterialExpressionReroute_38}]}
  }
  MaterialExpressionPinBase_22 {
   class=MaterialExpressionPinBase
   inputs []
   properties {reroutePins [
     UV {expression=&self:MaterialExpressionReroute_15}
     CutOff {expression=&self:MaterialExpressionReroute_16}
    ]}
  }
  MaterialExpressionPinBase_23 {
   class=MaterialExpressionPinBase
   inputs [Output {from=MaterialExpressionMultiply_45, output=""}]
   properties {reroutePins [Output {expression=&self:MaterialExpressionReroute_17}]}
  }
  MaterialExpressionPinBase_24 {
   class=MaterialExpressionPinBase
   inputs []
   properties {reroutePins [
     UV {expression=&self:MaterialGraph_0.MaterialGraphNode_Composite_14.BarMask.MaterialExpressionReroute_3}
     CutOff {expression=&self:MaterialGraph_0.MaterialGraphNode_Composite_14.BarMask.MaterialExpressionReroute_4}
    ]}
  }
  MaterialExpressionPinBase_25 {
   class=MaterialExpressionPinBase
   inputs [Output {from=MaterialExpressionMultiply_46, output=""}]
   properties {reroutePins [Output {expression=&self:MaterialGraph_0.MaterialGraphNode_Composite_14.BarMask.MaterialExpressionReroute_5}]}
  }
  MaterialExpressionPinBase_26 {
   class=MaterialExpressionPinBase
   inputs []
   properties {reroutePins [
     UV {expression=&self:MaterialExpressionReroute_11}
     CutOff {expression=&self:MaterialExpressionReroute_12}
    ]}
  }
  MaterialExpressionPinBase_27 {
   class=MaterialExpressionPinBase
   inputs [Output {from=MaterialExpressionMultiply_47, output=""}]
   properties {reroutePins [Output {expression=&self:MaterialExpressionReroute_13}]}
  }
  MaterialExpressionPinBase_28 {
   class=MaterialExpressionPinBase
   inputs []
   properties {reroutePins [
     UV {expression=&self:MaterialExpressionReroute_27}
     CutOff {expression=&self:MaterialExpressionReroute_28}
    ]}
  }
  MaterialExpressionPinBase_29 {
   class=MaterialExpressionPinBase
   inputs [Output {from=MaterialExpressionMultiply_48, output=""}]
   properties {reroutePins [Output {expression=&self:MaterialExpressionReroute_29}]}
  }
  MaterialExpressionPinBase_3 {class=MaterialExpressionPinBase, inputs [], properties {}}
  MaterialExpressionPinBase_30 {
   class=MaterialExpressionPinBase
   inputs []
   properties {reroutePins [
     UV {expression=&self:MaterialExpressionReroute_44}
     CutOff {expression=&self:MaterialExpressionReroute_45}
    ]}
  }
  MaterialExpressionPinBase_31 {
   class=MaterialExpressionPinBase
   inputs [Output {from=MaterialExpressionMultiply_53, output=""}]
   properties {reroutePins [Output {expression=&self:MaterialExpressionReroute_46}]}
  }
  MaterialExpressionPinBase_32 {
   class=MaterialExpressionPinBase
   inputs []
   properties {reroutePins [
     UV {expression=&self:MaterialExpressionReroute_52}
     CutOff {expression=&self:MaterialExpressionReroute_53}
    ]}
  }
  MaterialExpressionPinBase_33 {
   class=MaterialExpressionPinBase
   inputs [Output {from=MaterialExpressionMultiply_57, output=""}]
   properties {reroutePins [Output {expression=&self:MaterialExpressionReroute_54}]}
  }
  MaterialExpressionPinBase_34 {
   class=MaterialExpressionPinBase
   inputs []
   properties {reroutePins [
     UV {expression=null}
     CutOff {expression=null}
    ]}
  }
  MaterialExpressionPinBase_35 {class=MaterialExpressionPinBase, inputs [], properties {reroutePins [Output {expression=null}]}}
  MaterialExpressionPinBase_4 {class=MaterialExpressionPinBase, inputs [], properties {}}
  MaterialExpressionPinBase_5 {class=MaterialExpressionPinBase, inputs [], properties {}}
  MaterialExpressionPinBase_6 {class=MaterialExpressionPinBase, inputs [], properties {reroutePins [InputPin {expression=null}]}}
  MaterialExpressionPinBase_7 {class=MaterialExpressionPinBase, inputs [], properties {reroutePins [Output {expression=null}]}}
  MaterialExpressionPinBase_8 {
   class=MaterialExpressionPinBase
   inputs []
   properties {reroutePins [
     UV {expression=&self:MaterialGraph_0.MaterialGraphNode_Composite_10.BarMask.MaterialExpressionReroute_18}
     CutOff {expression=&self:MaterialGraph_0.MaterialGraphNode_Composite_10.BarMask.MaterialExpressionReroute_19}
    ]}
  }
  MaterialExpressionPinBase_9 {
   class=MaterialExpressionPinBase
   inputs [Output {from=MaterialExpressionMultiply_33, output=""}]
   properties {reroutePins [Output {expression=&self:MaterialGraph_0.MaterialGraphNode_Composite_10.BarMask.MaterialExpressionReroute_20}]}
  }
  MaterialExpressionPower_0 {
   class=MaterialExpressionPower
   inputs [
    Base {from=MaterialExpressionAbs_0}
    Exp {from=MaterialExpressionScalarParameter_22}
   ]
  }
  MaterialExpressionReroute_0 {class=MaterialExpressionReroute, inputs [None {from=MaterialExpressionNamedRerouteUsage_6}]}
  MaterialExpressionReroute_10 {class=MaterialExpressionReroute, inputs [None {from=MaterialExpressionReroute_3, output=UV}]}
  MaterialExpressionReroute_11 {class=MaterialExpressionReroute, inputs [None {from=MaterialExpressionReroute_41}]}
  MaterialExpressionReroute_12 {class=MaterialExpressionReroute, inputs [None {from=MaterialExpressionAdd_13}]}
  MaterialExpressionReroute_13 {class=MaterialExpressionReroute, inputs [None {from=MaterialExpressionMultiply_47}]}
  MaterialExpressionReroute_14 {class=MaterialExpressionReroute, inputs [None {from=MaterialExpressionReroute_9, output=UV}]}
  MaterialExpressionReroute_15 {class=MaterialExpressionReroute, inputs [None {from=MaterialExpressionSubtract_5}]}
  MaterialExpressionReroute_16 {class=MaterialExpressionReroute, inputs [None {from=MaterialExpressionReroute_40}]}
  MaterialExpressionReroute_17 {class=MaterialExpressionReroute, inputs [None {from=MaterialExpressionMultiply_45}]}
  MaterialExpressionReroute_19 {class=MaterialExpressionReroute, inputs [None {from=MaterialExpressionScalarParameter_15}]}
  MaterialExpressionReroute_20 {class=MaterialExpressionReroute, inputs [None {from=MaterialExpressionSubtract_1}]}
  MaterialExpressionReroute_27 {class=MaterialExpressionReroute, inputs [None {from=MaterialExpressionSubtract_4}]}
  MaterialExpressionReroute_28 {class=MaterialExpressionReroute, inputs [None {from=MaterialExpressionReroute_40}]}
  MaterialExpressionReroute_29 {class=MaterialExpressionReroute, inputs [None {from=MaterialExpressionMultiply_48}]}
  MaterialExpressionReroute_30 {class=MaterialExpressionReroute, inputs [None {from=MaterialExpressionReroute_0, output=UV}]}
  MaterialExpressionReroute_32 {class=MaterialExpressionReroute, inputs [None {from=MaterialExpressionReroute_16, output=CutOffMax}]}
  MaterialExpressionReroute_35 {class=MaterialExpressionReroute, inputs [None {from=MaterialExpressionFrac_1}]}
  MaterialExpressionReroute_36 {class=MaterialExpressionReroute, inputs [None {from=MaterialExpressionConstant4Vector_0, output=B}]}
  MaterialExpressionReroute_37 {class=MaterialExpressionReroute, inputs [None {from=MaterialExpressionConstant4Vector_0, output=A}]}
  MaterialExpressionReroute_38 {class=MaterialExpressionReroute, inputs [None {from=MaterialExpressionMultiply_39}]}
  MaterialExpressionReroute_39 {class=MaterialExpressionReroute, inputs [None {from=MaterialExpressionReroute_36, output=CutOffMax}]}
  MaterialExpressionReroute_40 {class=MaterialExpressionReroute, inputs [None {from=MaterialExpressionScalarParameter_18}]}
  MaterialExpressionReroute_41 {class=MaterialExpressionReroute, inputs [None {from=MaterialExpressionSubtract_4}]}
  MaterialExpressionReroute_43 {class=MaterialExpressionReroute, inputs [None {from=MaterialExpressionReroute_15, output=UV}]}
  MaterialExpressionReroute_44 {class=MaterialExpressionReroute, inputs [None {from=MaterialExpressionReroute_62}]}
  MaterialExpressionReroute_45 {class=MaterialExpressionReroute, inputs [None {from=MaterialExpressionConstant_6}]}
  MaterialExpressionReroute_46 {class=MaterialExpressionReroute, inputs [None {from=MaterialExpressionMultiply_53}]}
  MaterialExpressionReroute_47 {class=MaterialExpressionReroute, inputs [None {from=MaterialExpressionReroute_3, output=UV}]}
  MaterialExpressionReroute_48 {class=MaterialExpressionReroute, inputs [None {from=MaterialExpressionReroute_11, output=UV}]}
  MaterialExpressionReroute_49 {class=MaterialExpressionReroute, inputs [None {from=MaterialExpressionReroute_27, output=UV}]}
  MaterialExpressionReroute_50 {class=MaterialExpressionReroute, inputs [None {from=MaterialExpressionRotator_0}]}
  MaterialExpressionReroute_51 {class=MaterialExpressionReroute, inputs [None {from=MaterialExpressionReroute_44, output=UV}]}
  MaterialExpressionReroute_52 {class=MaterialExpressionReroute, inputs [None {from=MaterialExpressionReroute_62}]}
  MaterialExpressionReroute_53 {class=MaterialExpressionReroute, inputs [None {from=MaterialExpressionConstant_7}]}
  MaterialExpressionReroute_54 {class=MaterialExpressionReroute, inputs [None {from=MaterialExpressionMultiply_57}]}
  MaterialExpressionReroute_55 {class=MaterialExpressionReroute, inputs [None {from=MaterialExpressionReroute_52, output=UV}]}
  MaterialExpressionReroute_6 {class=MaterialExpressionReroute, inputs [None {from=MaterialExpressionReroute_18, output=UV}]}
  MaterialExpressionReroute_62 {class=MaterialExpressionReroute, inputs [None {from=MaterialExpressionReroute_63}]}
  MaterialExpressionReroute_63 {class=MaterialExpressionReroute, inputs [None {from=MaterialExpressionAppendVector_7}]}
  MaterialExpressionReroute_64 {class=MaterialExpressionReroute, inputs [None {from=MaterialExpressionAppendVector_6}]}
  MaterialExpressionReroute_65 {class=MaterialExpressionReroute, inputs [None {from=MaterialExpressionReroute_64}]}
  MaterialExpressionReroute_66 {class=MaterialExpressionReroute, inputs [None {from=MaterialExpressionReroute_64}]}
  MaterialExpressionRotator_0 {
   class=MaterialExpressionRotator
   inputs [
    Coordinate {from=MaterialExpressionAppendVector_7, output=""}
    Time {from=MaterialExpressionConstant_5, output=""}
   ]
   properties {}
  }
  MaterialExpressionRotator_1 {
   class=MaterialExpressionRotator
   inputs [
    Coordinate {from=MaterialExpressionMaterialFunctionCall_24, output=UVs}
    Time {from=MaterialExpressionMultiply_58, output=""}
   ]
   properties {}
  }
  MaterialExpressionSaturate_0 {class=MaterialExpressionSaturate, inputs [None {from=MaterialExpressionAdd_0}]}
  MaterialExpressionScalarParameter_0 {class=MaterialExpressionScalarParameter, properties {defaultValue=0.85, parameterName=Brightness}}
  MaterialExpressionScalarParameter_11 {class=MaterialExpressionScalarParameter, properties {defaultValue=0.5, parameterName=Oscillate_Speed}}
  MaterialExpressionScalarParameter_12 {class=MaterialExpressionScalarParameter, properties {defaultValue=-0.75, group=Triangles}}
  MaterialExpressionScalarParameter_13 {class=MaterialExpressionScalarParameter, properties {defaultValue=0.75, group=Triangles, parameterName=Param2}}
  MaterialExpressionScalarParameter_15 {class=MaterialExpressionScalarParameter, properties {defaultValue=0.25, group=Triangles, parameterName=LeftTri_Size}}
  MaterialExpressionScalarParameter_18 {class=MaterialExpressionScalarParameter, properties {defaultValue=0.2, group=Triangles, parameterName=RightTri_Size}}
  MaterialExpressionScalarParameter_19 {class=MaterialExpressionScalarParameter, properties {group=UVs, parameterName=UV_Angle, sortPriority=10}}
  MaterialExpressionScalarParameter_20 {class=MaterialExpressionScalarParameter, properties {defaultValue=1, group=Triangles, parameterName=RightTriangle_Presence}}
  MaterialExpressionScalarParameter_22 {class=MaterialExpressionScalarParameter, properties {defaultValue=5, group=Background, parameterName=BG_Pow}}
  MaterialExpressionScalarParameter_23 {class=MaterialExpressionScalarParameter, properties {defaultValue=1, group=UVs, parameterName=UVScale_V, sortPriority=2}}
  MaterialExpressionScalarParameter_24 {class=MaterialExpressionScalarParameter, properties {defaultValue=1, group=UVs, parameterName=UVScale_Uniform, sortPriority=0}}
  MaterialExpressionScalarParameter_25 {class=MaterialExpressionScalarParameter, properties {defaultValue=0.65, group=Background, parameterName=Amplitude_Presence, sliderMax=1}}
  MaterialExpressionScalarParameter_28 {class=MaterialExpressionScalarParameter, properties {defaultValue=0.65, group=Background, parameterName=Amplitude_Presence, sliderMax=1}}
  MaterialExpressionScalarParameter_29 {class=MaterialExpressionScalarParameter, properties {defaultValue=0.5, parameterName=Oscillate_Speed}}
  MaterialExpressionScalarParameter_36 {class=MaterialExpressionScalarParameter, properties {defaultValue=-0.5, parameterName=Param_1}}
  MaterialExpressionScalarParameter_37 {class=MaterialExpressionScalarParameter, properties {defaultValue=0.05, parameterName=ss}}
  MaterialExpressionScalarParameter_38 {class=MaterialExpressionScalarParameter, properties {defaultValue=0.75, parameterName=Param_2}}
  MaterialExpressionScalarParameter_41 {class=MaterialExpressionScalarParameter, properties {defaultValue=0.5, parameterName=Distortion}}
  MaterialExpressionScalarParameter_43 {class=MaterialExpressionScalarParameter, properties {defaultValue=1, parameterName=Specular}}
  MaterialExpressionScalarParameter_44 {class=MaterialExpressionScalarParameter, properties {defaultValue=0.125, parameterName=Roughness}}
  MaterialExpressionScalarParameter_6 {class=MaterialExpressionScalarParameter, properties {defaultValue=1, group=UVs, parameterName=UVScale_U, sortPriority=1}}
  MaterialExpressionScalarParameter_7 {class=MaterialExpressionScalarParameter, properties {group=UVs, parameterName=UVOffset_X, sortPriority=20}}
  MaterialExpressionScalarParameter_8 {class=MaterialExpressionScalarParameter, properties {group=UVs, parameterName=UVOffset_Y, sortPriority=21}}
  MaterialExpressionScalarParameter_9 {class=MaterialExpressionScalarParameter, properties {defaultValue=0.75, parameterName=Texture_Presence}}
  MaterialExpressionSine_0 {class=MaterialExpressionSine, inputs [None {from=MaterialExpressionMultiply_9}]}
  MaterialExpressionSine_1 {class=MaterialExpressionSine, inputs [None {from=MaterialExpressionComponentMask_40}], properties {period=2}}
  MaterialExpressionStep_0 {
   class=MaterialExpressionStep
   inputs [
    Y {from=MaterialExpressionComponentMask_0, output=""}
    X {from="", output=""}
   ]
   properties {constX=0.5}
  }
  MaterialExpressionStep_1 {
   class=MaterialExpressionStep
   inputs [
    Y {from="", output=""}
    X {from=MaterialExpressionComponentMask_1, output=""}
   ]
   properties {constX=0.5, constY=0.35}
  }
  MaterialExpressionStep_16 {
   class=MaterialExpressionStep
   inputs [
    Y {from=MaterialExpressionComponentMask_11, output=""}
    X {from=MaterialExpressionReroute_19, output=CutOff}
   ]
   properties {constX=0.3}
  }
  MaterialExpressionStep_17 {
   class=MaterialExpressionStep
   inputs [
    Y {from="", output=""}
    X {from=MaterialExpressionComponentMask_11, output=""}
   ]
   properties {constX=0.6, constY=-1}
  }
  MaterialExpressionStep_18 {
   class=MaterialExpressionStep
   inputs [
    Y {from=MaterialExpressionComponentMask_13, output=""}
    X {from=MaterialExpressionReroute_4, output=CutOff}
   ]
   properties {constX=0.3}
  }
  MaterialExpressionStep_19 {
   class=MaterialExpressionStep
   inputs [
    Y {from="", output=""}
    X {from=MaterialExpressionComponentMask_13, output=""}
   ]
   properties {constX=0.6, constY=-1}
  }
  MaterialExpressionStep_20 {
   class=MaterialExpressionStep
   inputs [
    Y {from=MaterialExpressionComponentMask_15, output=""}
    X {from=MaterialExpressionReroute_10, output=CutOff}
   ]
   properties {constX=0.3}
  }
  MaterialExpressionStep_21 {
   class=MaterialExpressionStep
   inputs [
    Y {from="", output=""}
    X {from=MaterialExpressionComponentMask_15, output=""}
   ]
   properties {constX=0.6, constY=-1}
  }
  MaterialExpressionStep_24 {
   class=MaterialExpressionStep
   inputs [
    Y {from=MaterialExpressionComponentMask_20, output=""}
    X {from=MaterialExpressionReroute_32, output=""}
   ]
   properties {constX=0.3}
  }
  MaterialExpressionStep_25 {
   class=MaterialExpressionStep
   inputs [
    Y {from=MaterialExpressionReroute_17, output=CutOffMin}
    X {from=MaterialExpressionComponentMask_20, output=""}
   ]
   properties {constX=0.6, constY=-0.5}
  }
  MaterialExpressionStep_26 {
   class=MaterialExpressionStep
   inputs [
    Y {from=MaterialExpressionComponentMask_22, output=""}
    X {from=MaterialExpressionReroute_1, output=CutOff}
   ]
   properties {constX=0.3}
  }
  MaterialExpressionStep_27 {
   class=MaterialExpressionStep
   inputs [
    Y {from="", output=""}
    X {from=MaterialExpressionComponentMask_22, output=""}
   ]
   properties {constX=0.6, constY=-1}
  }
  MaterialExpressionStep_28 {
   class=MaterialExpressionStep
   inputs [
    Y {from=MaterialExpressionComponentMask_24, output=""}
    X {from=MaterialExpressionReroute_39, output=""}
   ]
   properties {constX=0.3}
  }
  MaterialExpressionStep_29 {
   class=MaterialExpressionStep
   inputs [
    Y {from=MaterialExpressionReroute_37, output=CutOffMin}
    X {from=MaterialExpressionComponentMask_24, output=""}
   ]
   properties {constX=0.6, constY=-0.5}
  }
  MaterialExpressionStep_30 {
   class=MaterialExpressionStep
   inputs [
    Y {from=MaterialExpressionComponentMask_26, output=""}
    X {from=MaterialExpressionReroute_16, output=CutOff}
   ]
   properties {constX=0.3}
  }
  MaterialExpressionStep_31 {
   class=MaterialExpressionStep
   inputs [
    Y {from="", output=""}
    X {from=MaterialExpressionComponentMask_26, output=""}
   ]
   properties {constX=0.6, constY=-0.5}
  }
  MaterialExpressionStep_32 {
   class=MaterialExpressionStep
   inputs [
    Y {from=MaterialExpressionComponentMask_28, output=""}
    X {from=MaterialExpressionReroute_4, output=CutOff}
   ]
   properties {constX=0.3}
  }
  MaterialExpressionStep_33 {
   class=MaterialExpressionStep
   inputs [
    Y {from="", output=""}
    X {from=MaterialExpressionComponentMask_28, output=""}
   ]
   properties {constX=0.6, constY=-0.5}
  }
  MaterialExpressionStep_34 {
   class=MaterialExpressionStep
   inputs [
    Y {from=MaterialExpressionComponentMask_30, output=""}
    X {from=MaterialExpressionReroute_12, output=CutOff}
   ]
   properties {constX=0.3}
  }
  MaterialExpressionStep_35 {
   class=MaterialExpressionStep
   inputs [
    Y {from="", output=""}
    X {from=MaterialExpressionComponentMask_30, output=""}
   ]
   properties {constX=0.6, constY=-0.5}
  }
  MaterialExpressionStep_36 {
   class=MaterialExpressionStep
   inputs [
    Y {from=MaterialExpressionComponentMask_32, output=""}
    X {from=MaterialExpressionReroute_28, output=CutOff}
   ]
   properties {constX=0.3}
  }
  MaterialExpressionStep_37 {
   class=MaterialExpressionStep
   inputs [
    Y {from="", output=""}
    X {from=MaterialExpressionComponentMask_32, output=""}
   ]
   properties {constX=0.6, constY=-0.5}
  }
  MaterialExpressionStep_38 {
   class=MaterialExpressionStep
   inputs [
    Y {from=MaterialExpressionComponentMask_35, output=""}
    X {from=MaterialExpressionReroute_45, output=CutOff}
   ]
   properties {constX=0.3}
  }
  MaterialExpressionStep_39 {
   class=MaterialExpressionStep
   inputs [
    Y {from="", output=""}
    X {from=MaterialExpressionComponentMask_35, output=""}
   ]
   properties {constX=0.6, constY=-0.5}
  }
  MaterialExpressionStep_40 {
   class=MaterialExpressionStep
   inputs [
    Y {from=MaterialExpressionComponentMask_37, output=""}
    X {from=MaterialExpressionReroute_53, output=CutOff}
   ]
   properties {constX=0.3}
  }
  MaterialExpressionStep_41 {
   class=MaterialExpressionStep
   inputs [
    Y {from="", output=""}
    X {from=MaterialExpressionComponentMask_37, output=""}
   ]
   properties {constX=0.6, constY=-0.5}
  }
  MaterialExpressionSubtract_0 {
   class=MaterialExpressionSubtract
   inputs [
    A {from=MaterialExpressionAdd_1}
    B {from=MaterialExpressionDivide_1}
   ]
  }
  MaterialExpressionSubtract_1 {
   class=MaterialExpressionSubtract
   inputs [
    A {from=MaterialExpressionAdd_6}
    B {from=MaterialExpressionDivide_0}
   ]
  }
  MaterialExpressionSubtract_2 {
   class=MaterialExpressionSubtract
   inputs [
    A {from=MaterialExpressionAdd_3}
    B {from=MaterialExpressionDivide_2}
   ]
  }
  MaterialExpressionSubtract_3 {
   class=MaterialExpressionSubtract
   inputs [
    A {from=MaterialExpressionAdd_8}
    B {from=MaterialExpressionDivide_3}
   ]
  }
  MaterialExpressionSubtract_4 {
   class=MaterialExpressionSubtract
   inputs [
    A {from=MaterialExpressionAdd_11}
    B {from=MaterialExpressionDivide_4}
   ]
  }
  MaterialExpressionSubtract_5 {
   class=MaterialExpressionSubtract
   inputs [
    A {from=MaterialExpressionAdd_12}
    B {from=MaterialExpressionDivide_5}
   ]
  }
  MaterialExpressionTextureCoordinate_0 {class=MaterialExpressionTextureCoordinate}
  MaterialExpressionTextureObject_0 {class=MaterialExpressionTextureObject, inputs [], properties {texture=&AtmosphericCloudNoise01}}
  MaterialExpressionTextureSample_1 {
   class=MaterialExpressionTextureSample
   inputs [
    UVs {from=MaterialExpressionAdd_9}
    Tex {from=MaterialExpressionTextureObject_0}
    Apply View MipBias {from=""}
   ]
  }
  MaterialExpressionTextureSample_3 {
   class=MaterialExpressionTextureSample
   inputs [
    UVs {from=MaterialExpressionPanner_3}
    Tex {from=MaterialExpressionTextureObject_0}
    Apply View MipBias {from=""}
   ]
  }
  MaterialExpressionTime_0 {class=MaterialExpressionTime}
  MaterialExpressionTime_1 {class=MaterialExpressionTime}
  MaterialExpressionTime_2 {class=MaterialExpressionTime}
  MaterialExpressionTime_3 {class=MaterialExpressionTime}
 }
 properties {bAutomaticallySetUsageInEditor=false, usedShadingModels=Default Lit}
}
