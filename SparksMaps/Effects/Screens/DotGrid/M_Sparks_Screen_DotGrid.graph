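# Audio-reactive material for Sparks stage screens, drawing an animated dot grid in stage colors through a shared screen-details function.
# Panning-noise smoke, angular-gradient rings and a noise-based fake FFT build the background and dot-size masks.
# Rotated, floor-quantized UVs give per-cell sampling coordinates, while their fractional counterpart tiles the dots.
# Audio collection beat and amplitude supply reactivity signals, and thresholded Primary color channels become configuration switches for layout variation.
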Material /SparksMaps/Effects/Screens/DotGrid/M_Sparks_Screen_DotGrid
graph {
 properties {bAutomaticallySetUsageInEditor=false, usedShadingModels=Default Lit}
 nodes [
  n1 = MaterialFunctionCall!(CustomUPosition) {materialFunction=/SparksStageShow/Systems/StageManager/FXManager/MaterialFunctions/MF_MojoStageColors}
  n2 = LinearInterpolate(B=NamedRerouteUsage(), Alpha=ScalarParameter() {defaultValue=0.94, parameterName=Bg_Contrast}) {constA=0.5, constB=0.5}
  n3 = MaterialFunctionCall() {materialFunction=&MF_MojoStageColors}
  n4 = ScalarParameter() {defaultValue=1, parameterName=DotBrightness}
  n5 = Multiply(
   A=LinearInterpolate(A=LinearInterpolate(A=@n1.Primary, B=@n1.Primary, Alpha=@n2), B=Multiply(A=@n3.Secondary, B=@n4), Alpha=NamedRerouteUsage())
   B=ScalarParameter() {defaultValue=0.3, parameterName=MasterBrightness}
  )
  n6 = NamedRerouteUsage()
  MaterialFunctionCall!(
   Emissive=LinearInterpolate(A=@n5, B=Multiply(A=@n5, B=@n6), Alpha=ScalarParameter() {defaultValue=0.5, parameterName=BeatData_Presence})
   Scanlines Tile
  ) {materialFunction=/SparksMaps/Effects/Screens/Materials/MF/MF_Sparks_Screen_Details}
  n7 = Panner(
   Coordinate=SphereMask(
    A=NamedRerouteUsage()
    B=Constant() {r=0.5}
    Radius=ScalarParameter() {defaultValue=0.5, parameterName=BGRadius}
    Hardness=ScalarParameter() {parameterName=Hardness}
   )
   Speed=ScalarParameter() {defaultValue=0.36, parameterName=BG_PanSpeed}
  ) {speedX=0.2}
  n8 = Sine(None=Frac(None=ComponentMask(None=@n7) {r=true}))
  n9 = LinearInterpolate(B=NamedRerouteUsage(), Alpha=ScalarParameter() {defaultValue=1.25, parameterName=BG_SmokePresence})
  Multiply(A=@n8, B=@n9)
  n10 = CollectionParameter() {collection=/FMAudioReactiveData/AudioAnalysis/MPC_Sparks_Audio_Parameters, parameterName=Amplitude}
  Multiply(A=@n10, B=@n10)
  NamedRerouteDeclaration(None=@n10) {name=AmplitudeSignal, nodeColor=<rgba 0.1953 1 0 1>}
  n11 = CurveAtlasRowParameter(CurveTime=CollectionParameter() {collection=&MPC_Sparks_Audio_Parameters, parameterName=BeatLifetime}) {
   atlas=/SparksMaps/Effects/Screens/Curves/CurveAtlas_Audio_Beat_A
   curve=/SparksMaps/Effects/Screens/Curves/Curve_Audio_Beat_Remap_001
   parameterName=ReactivityCurve
  }
  MaterialExpressionReroute_3 = Reroute(None=Multiply(A=@n11, B=@n11))
  NamedRerouteDeclaration(None=MaterialFunctionCall(
    Input=@MaterialExpressionReroute_3
    Input Low=Constant()
    Input High=Constant() {r=1}
    Target Low=ScalarParameter() {defaultValue=0.25, group=AudioSignals, parameterName=BG Beat Low}
    Target High=Constant() {r=1.25}
   ) {materialFunction=&RemapValueRange}.Result) {name=BeatSignal, nodeColor=<rgba 0 1 0.2266 1>}
  n12 = NamedRerouteUsage()
  n13 = ScalarParameter() {defaultValue=2, parameterName=BG_RingsPresence}
  n14 = Multiply(A=Add(A=@n8, B=@n9), B=OneMinus(None=LinearInterpolate(B=@n12, Alpha=@n13)))
  n15 = ScalarParameter() {defaultValue=1, parameterName=BGPow}
  NamedRerouteDeclaration(None=Saturate(None=Power(Base=@n14, Exp=@n15))) {name=BG_Mask, nodeColor=<rgba 0 0.9531 1 1>}
  MaterialExpressionReroute_2 = Reroute(None=ScalarParameter() {defaultValue=0.025, parameterName=DotSize_Min})
  n16 = LinearInterpolate(
   B=NamedRerouteUsage()
   Alpha=LinearInterpolate(
    A=ScalarParameter() {defaultValue=1, parameterName=DotSize_FFT_PresenceMin}
    B=ScalarParameter() {defaultValue=2, parameterName=DotSize_FFT_PresenceMax}
    Alpha=NamedRerouteUsage()
   )
  )
  n17 = Add(
   A=LinearInterpolate(B=NamedRerouteUsage(), Alpha=ScalarParameter() {defaultValue=1.5, parameterName=DotSize_Smoke_Presence})
   B=LinearInterpolate(B=NamedRerouteUsage(), Alpha=ScalarParameter() {defaultValue=0.75, parameterName=DotSize_Rings_Presence})
  )
  n18 = Multiply(A=Add(A=@MaterialExpressionReroute_2, B=Add(A=@n16, B=@n17)), B=Constant() {desc=DotSize_Multiplier, r=0.46}) {constB=0.5}
  n19 = LinearInterpolate(Alpha=NamedRerouteUsage()) {constA=1, constB=0.85}
  NamedRerouteDeclaration(None=Saturate(None=Power(Base=@n18, Exp=@n19))) {name=DotSize_MaskOutput, nodeColor=<rgba 1 0.2344 0 1>}
  n20 = MaterialFunctionCall(Color=MaterialFunctionCall() {materialFunction=&MF_MojoStageColors}.Primary) {materialFunction=&SplitComponents}
  NamedRerouteDeclaration(None=Step(Y=@n20.R) {constX=0.5}) {name=ConfigurationSwitch01, nodeColor=<rgba 0 1 0.4141 1>}
  NamedRerouteDeclaration(None=Step(Y=@n20.B) {constX=0.5}) {name=ConfigurationSwitch03, nodeColor=<rgba 1 0.0938 0 1>}
  NamedRerouteDeclaration(None=Step(Y=@n20.G) {constX=0.5}) {name=ConfigurationSwitch02, nodeColor=<rgba 0.6406 1 0 1>}
  n21 = LinearInterpolate(
   A=ScalarParameter() {defaultValue=32, parameterName=DotScreen_ScaleMax}
   B=ScalarParameter() {defaultValue=22, parameterName=DotScreen_ScaleMin}
   Alpha=NamedRerouteUsage()
  ) {constA=32, constB=24}
  MaterialExpressionReroute_5 = Reroute(None=@n21)
  n22 = NamedRerouteUsage()
  n23 = LinearInterpolate(Alpha=@n22) {constA=90, constB=0}
  n24 = LinearInterpolate(Alpha=@n22) {constA=45, constB=-45}
  n25 = NamedRerouteUsage()
  n26 = ComponentMask(None=Divide(A=LinearInterpolate(A=@n23, B=@n24, Alpha=@n25)) {constB=14.324, desc=Convert to Degrees}) {r=true}
  n27 = ScalarParameter() {defaultValue=0.05, parameterName=DotPattern_U_Pan}
  n28 = ScalarParameter() {parameterName=DotPattern_V_Pan}
  n29 = Panner(
   Coordinate=Rotator(Coordinate=NamedRerouteUsage(), Time=@n26)
   Speed=LinearInterpolate(A=Constant(), B=AppendVector(A=@n27, B=@n28), Alpha=NamedRerouteUsage())
  )
  n30 = Floor(None=Multiply(A=@n29, B=@MaterialExpressionReroute_5))
  NamedRerouteDeclaration(None=Rotator(Coordinate=Divide(A=@n30, B=@MaterialExpressionReroute_5), Time=Multiply(A=@n26) {constB=-1})) {name=RotQuantUVs, nodeColor=<rgba 1 0 0.0234 1>}
  NamedRerouteDeclaration(None=Frac(None=Multiply(A=@n29, B=@n21))) {name=Frac_DotScreenUVs, nodeColor=<rgba 0.4297 1 0 1>}
  n31 = NamedRerouteUsage()
  n32 = TextureObject() {samplerType=SAMPLERTYPE_LinearColor, texture=/Game/Effects/Fort_Effects/Gameplay/SmokeDustExplosions/Textures/T_NoisePack_LinearColor}
  n33 = Multiply(A=@n31) {constB=0.75}
  n34 = Multiply(
   A=TextureSample(UVs=Panner(Coordinate=@n31) {speedX=0.2, speedY=0.2}, Tex=@n32).R
   B=TextureSample(UVs=Panner(Coordinate=@n33) {speedX=-0.1, speedY=-0.1}, Tex=@n32).G
  )
  n35 = ScalarParameter() {defaultValue=2, parameterName=SmokePow}
  NamedRerouteDeclaration(None=Saturate(None=Power(Base=@n34, Exp=@n35))) {name=Smoke_Mask, nodeColor=<rgba 1 0 0.1172 1>}
  n36 = TextureCoordinate()
  n37 = ScalarParameter() {defaultValue=1, group=UVs, parameterName=UVScale}
  n38 = ScalarParameter() {group=UVs, parameterName=UV_X_Offset}
  n39 = ScalarParameter() {group=UVs, parameterName=UV_Y_Offset}
  NamedRerouteDeclaration(None=Add(
    A=MaterialFunctionCall(UVs=@n36, Texture Scale=@n37) {materialFunction=&ScaleUVsByCenter}.UVs
    B=AppendVector(A=@n38, B=@n39)
   )) {name=UVs, nodeColor=<rgba 1 0 0.0222 1>}
  n40 = Constant() {r=0.5}
  n41 = NamedRerouteUsage()
  n42 = NamedRerouteUsage()
  n43 = ScalarParameter() {defaultValue=1, parameterName=DotHardness}
  NamedRerouteDeclaration(None=Saturate(None=SphereMask(A=@n40, B=@n41, Radius=@n42, Hardness=@n43) {attenuationRadius=0.5, hardnessPercent=0})) {name=DotScreen, nodeColor=<rgba 1 0 0.5391 1>}
  n44 = NamedRerouteUsage()
  n45 = Multiply(A=MaterialFunctionCall!(UV=Rotator(Coordinate=@n44), Rotate) {materialFunction=/AudioWidgets/AudioMaterialSlate/Functions/MF_AngleGradient}.Result) {constB=2}
  n46 = ComponentMask(None=Abs(None=Add(A=@n45) {constB=-1})) {r=true}
  n47 = Constant2Vector()
  n48 = Constant()
  n49 = Panner(Coordinate=MaterialFunctionCall(UVs=Add(A=@n44, B=@n47), Size=@n48, GlowMin=@n48, GlowMax=Constant() {r=1.35}) {materialFunction=&MF_UI_Circle}.Fill) {speedX=0.5}
  n50 = CurveAtlasRowParameter(CurveTime=Frac(None=ComponentMask(None=@n49) {r=true})) {
   atlas=&/CA_S8_Gradients
   curve=&/C_S8_Screen_Gradient02
   defaultValue=1
   parameterName=Masks
  }
  n51 = NamedRerouteUsage()
  n52 = Multiply(A=MaterialFunctionCall(UV=Rotator(Coordinate=@n51) {speed=-0.15}) {materialFunction=&MF_AngleGradient}.Result) {constB=2}
  n53 = ComponentMask(None=Abs(None=Add(A=@n52) {constB=-1})) {r=true}
  n54 = Constant()
  n55 = Panner(Coordinate=MaterialFunctionCall(UVs=Add(A=@n47, B=@n51), Size=@n54, GlowMin=@n54, GlowMax=Constant() {r=1.35}) {materialFunction=&MF_UI_Circle}.Fill) {speedX=0.2}
  n56 = CurveAtlasRowParameter(CurveTime=Frac(None=ComponentMask(None=@n55) {r=true})) {
   atlas=&CA_S8_Gradients
   curve=&C_S8_Screen_Gradient02
   defaultValue=1
   parameterName=Masks
  }
  n57 = Add(A=Multiply(A=Power(Base=@n46) {constExponent=3}, B=@n50.R), B=Multiply(A=Power(Base=@n53) {constExponent=3}, B=@n56.R))
  n58 = ScalarParameter() {defaultValue=6, parameterName=Strength02}
  NamedRerouteDeclaration(None=Saturate(None=Multiply(A=@n57, B=@n58))) {name=Rings_Mask, nodeColor=<rgba 1 0.6563 0 1>}
  MaterialExpressionReroute_0 = Reroute(None=Multiply(A=NamedRerouteUsage(), B=Constant2Vector() {r=1}))
  n59 = Constant2Vector() {g=0.2, r=0.2}
  n60 = VectorParameter() {parameterName=FakeFFT_PanSpeed, defaultValue=<rgba 1 2 0 1>}
  n61 = AppendVector(A=@n60.R, B=@n60.R)
  n62 = TextureObject() {samplerType=SAMPLERTYPE_LinearColor, texture=&T_NoisePack_LinearColor}
  n63 = TextureSample(UVs=Panner(Coordinate=@MaterialExpressionReroute_0, Speed=Multiply(A=@n59, B=@n61)) {speedX=0.2, speedY=0.2}, Tex=@n62)
  n64 = Constant2Vector() {g=-0.2, r=-0.2}
  n65 = TextureSample(UVs=Panner(Coordinate=@MaterialExpressionReroute_0, Speed=Multiply(A=@n61, B=@n64)) {speedX=-0.2, speedY=-0.2}, Tex=@n62)
  n66 = NamedRerouteUsage()
  n67 = Power(Base=Multiply(A=Multiply(A=@n63.R, B=@n65.G), B=ComponentMask(None=@n66) {g=true}))
  n68 = LinearInterpolate(B=ScalarParameter() {defaultValue=2, parameterName=FakeFFT_Intensity}, Alpha=NamedRerouteUsage())
  NamedRerouteDeclaration(None=Saturate(None=Multiply(A=@n67, B=@n68))) {name=FakeFFT_Mask, nodeColor=<rgba 1 0 0.1172 1>}
  PinBase()
  PinBase!(Primary, Secondary, Tertiary) {reroutePins [
    Primary {expression=&self:MaterialGraph_0.MaterialGraphNode_Composite_1.Colors.MaterialExpressionReroute_0}
    Secondary {expression=&self:MaterialGraph_0.MaterialGraphNode_Composite_1.Colors.MaterialExpressionReroute_1}
    Tertiary {expression=&self:MaterialGraph_0.MaterialGraphNode_Composite_1.Colors.MaterialExpressionReroute_2}
   ]}
  PinBase!()
  PinBase!(Primary, Secondary, Tertiary) {reroutePins [
    Primary {expression=&self:MaterialGraph_0.MaterialGraphNode_Composite_2.Colors.MaterialExpressionReroute_3}
    Secondary {expression=&self:MaterialGraph_0.MaterialGraphNode_Composite_2.Colors.MaterialExpressionReroute_5}
    Tertiary {expression=&self:MaterialGraph_0.MaterialGraphNode_Composite_2.Colors.MaterialExpressionReroute_6}
   ]}
  ScalarParameter() {defaultValue=0.85, parameterName=DotSizePOW}
 ]
}
